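# Unlit, translucent, two-sided Niagara mesh particle material for a fresnel glow on the legs in the female Skirmish emote.
# Particle color and alpha are shaped by an optionally inverted fresnel, with dynamic parameters driving its exponent, multiplier and edge power, plus normal-based mesh inflation.
# Opacity adds an optional mesh-edge fade and a gradient along UV V.
# Depth fade softens intersections on all shading paths except mobile, which skips it.
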Material /Game/Effects/Fort_Effects/Emotes/Skirmish/Skirmish_Female/M_Skirmish_emote_Female_fresnel_legs
graph {
 properties {
  bUsedWithNiagaraMeshParticles=true
  blendMode=BLEND_Translucent
  pixelDepthOffsetMode=PDOM_Legacy
  shadingModel=MSM_Unlit
  twoSided=true
  usedShadingModels=Unlit
 }
 nodes [
  n1 = ParticleColor()
  n2 = DynamicParameter() {desc=Niagara_Dynamic Index2, parameterIndex=2, defaultValue=<rgba 0 4 5 2>, paramNames [InflateMesh,FresnelExp,FresnelMult,EdgePower]}
  n3 = Fresnel(ExponentIn=@n2.FresnelExp) {baseReflectFraction=0}
  n4 = Saturate(None=Power(Base=Multiply(A=@n3, B=@n2.FresnelMult), Exp=ScalarParameter() {defaultValue=1, group=Only_Use_Legs_fresnel, parameterName=Power}))
  n5 = StaticSwitchParameter(True=OneMinus(None=@n4), False=@n4) {defaultValue=true, group=Only_Use_Legs_fresnel, parameterName=InvertFresnel}
  n6 = Multiply(A=@n1.RGB, B=@n5)
  MaterialFunctionCall(Emissive=@n6) {materialFunction=&MF_FX_EmissiveAdjustments}
  Reroute(None=@n6)
  n7 = DynamicParameter() {desc=Niagara_Dynamic Index2, parameterIndex=2, defaultValue=<rgba 0 4 5 2>, paramNames [InflateMesh,FresnelExp,FresnelMult,EdgePower]}
  Reroute(None=Multiply(A=VertexNormalWS(), B=@n7.InflateMesh))
  n8 = Multiply(A=@n1.A, B=@n5)
  n9 = DynamicParameter() {desc=Niagara_Dynamic Index2, parameterIndex=2, defaultValue=<rgba 0 4 5 2>, paramNames [InflateMesh,FresnelExp,FresnelMult,EdgePower]}
  n10 = Power(Base=OneMinus(None=Fresnel() {baseReflectFraction=0, exponent=1}), Exp=@n9.EdgePower)
  n11 = Power(Base=ComponentMask(None=TextureCoordinate()) {g=true}) {constExponent=1.3}
  n12 = Multiply(
   A=LinearInterpolate(A=@n8, B=Multiply(A=@n8, B=@n10), Alpha=ScalarParameter() {group=Only_Use_Legs_fresnel, parameterName=Use_mesh_edge_fade})
   B=Saturate(None=Multiply(A=@n11) {constB=3})
  )
  n13 = DepthFade(Opacity=@n12, FadeDistance=ScalarParameter() {defaultValue=50, parameterName=DepthFade}) {fadeDistanceDefault=50}
  ShadingPathSwitch(Default=@n13, Deferred=@n13, Forward=@n13, Mobile=@n12)
 ]
}
